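# Masked parent material for the face accessory of the FareSpore Bulb player skin, with the shading model driven by the graph.
# It wraps skin effects between the shared 2023 character PreFX and PostFX functions using material attributes.
# PreFX feeds the FareSporeBulb Slurp function with pre-skinned position, Fresnel and the Tech Art Mask blue channel as alpha.
# The Slurp result passes through the FareSpore ears dithering function before PostFX.
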Material /BRCosmetics/Characters/Player/Male/Medium/Bodies/M_MED_FareSpore/Skins/Bulb/Materials/M_FareSporeBulb_FaceAcc_PARENT
graph {
 properties {
  bUseMaterialAttributes=true
  bUsedWithInstancedStaticMeshes=true
  bUsedWithMeshDeformer=true
  bUsedWithMorphTargets=true
  bUsedWithSkeletalMesh=true
  blendMode=BLEND_Masked
  pixelDepthOffsetMode=PDOM_Legacy
  shadingModel=MSM_FromMaterialExpression
  usedShadingModels=Default Lit | Subsurface | Clear Coat | Subsurface Profile | Cloth
 }
 nodes [
  n1 = MaterialFunctionCall() {materialFunction=&MF_CharacterParent_2023_PreFX}
  n2 = ComponentMask(None=@n1.Tech Art Mask) {b=true}
  n3 = NamedRerouteDeclaration(None=@n1.Pre skinned Local Position) {name=PreskinnedLocalPosition, nodeColor=<rgba 1 0 0 1>}
  MaterialFunctionCall(
   Material=MaterialFunctionCall!(MA=MaterialFunctionCall!(Alpha=@n2, AO, BubblesMask, SlurpNormalDistortionMask, Pre Skinned Position=@n3, Fresnel=@n1.Fresnel, Material=@n1.Fresnel) {materialFunction=&/MF_Slurp_FareSporeBulb}.Result) {materialFunction=/BRCosmetics/Characters/Player/Male/Medium/Bodies/M_MED_FareSpore/Materials/MF_FareSpore_EarsDithering}.Result
   M=@n1.M
   Pre Skinned Local Position=@n1.M
   Distance Blend=@n1.M
   Fresnel=@n1.M
  ) {materialFunction=&MF_CharacterParent_2023_PostFX}
 ]
}
